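# Animation Blueprint for the Animal Jackets backpack cosmetic, adding physics-driven secondary motion to its belly, feet and ears.
# The input pose runs through a chain of Anim Dynamics nodes with angular limits, damping and gravity overrides, with belly motion at half strength.
# Each update copies bIsTargeting into IsAim, which zeroes the alpha of the spring-driven ear dynamics while aiming and otherwise leaves it partially blended.
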AnimBlueprint /BRCosmetics/Accessories/FORT_Backpacks/M_MED_AnimalJackets_Backpack/Meshes/M_MED_AnimalJackets_backpack_AnimBP
target_skeleton=&/M_MED_AnimalJackets_backpack_MESH_Skeleton
graphs [
 AnimGraph {nodes [
   AnimGraphNode_SubInput_0 = Input Pose()
   n1 = Anim Dynamics - Bone: dyn_belly(ComponentPose=Local To Component(LocalPose=@AnimGraphNode_SubInput_0), Alpha=0.5, ExternalForce="0,0,0") {
    angularDampingOverride=0.999
    bGravityOverrideInSimSpace=true
    bOverrideAngularDamping=true
    bUseGravityOverride=true
    alphaBoolBlend {blendOption=Linear}
    boundBone {boneName=dyn_belly}
    componentAppliedLinearAccClamp=<300 300 300>
    componentLinearAccScale=<0.75 0.75 0.75>
    componentLinearVelScale=<0.75 0.75 0.75>
    gravityOverride {z=-981}
    physicsBodyDefinitions [{
      sphereCollisionRadius=0
      boundBone {boneName=dyn_belly}
      boxExtents=<10 33 33>
      constraintSetup {angularTargetAxis=AxisZ, angularLimitsMax=<5 15 10>, angularLimitsMin=<-5 -15 -1>, angularTarget=<0 0 0>}
      localJointOffset=<-8 -1 0>
     }]
   }
   n2 = Anim Dynamics - Bone: dyn_foot_r(ComponentPose=@n1, Alpha=1, ExternalForce="0,0,0") {
    angularDampingOverride=0.999
    bGravityOverrideInSimSpace=true
    bOverrideAngularDamping=true
    bUseGravityOverride=true
    alphaBoolBlend {blendOption=Linear}
    boundBone {boneName=dyn_foot_r}
    componentAppliedLinearAccClamp=<500 500 500>
    componentLinearAccScale=<1 1 1>
    componentLinearVelScale=<1 1 1>
    gravityOverride {z=-981}
    physicsBodyDefinitions [{
      sphereCollisionRadius=0
      boundBone {boneName=dyn_foot_r}
      boxExtents=<6 12 12>
      constraintSetup {angularLimitsMax=<5 12 12>, angularLimitsMin=<-5 -12 -12>, angularTarget=<0 0 0>}
      localJointOffset=<-6 -0.5 0>
     }]
   }
   n3 = Anim Dynamics - Bone: dyn_foot_l(ComponentPose=@n2, Alpha=1, ExternalForce="0,0,0") {
    angularDampingOverride=0.995
    angularSpringConstant=5
    bGravityOverrideInSimSpace=true
    bOverrideAngularDamping=true
    bUseGravityOverride=true
    alphaBoolBlend {blendOption=Linear}
    boundBone {boneName=dyn_foot_l}
    componentAppliedLinearAccClamp=<500 500 500>
    componentLinearAccScale=<1 1 1>
    componentLinearVelScale=<1 1 1>
    gravityOverride {z=-981}
    physicsBodyDefinitions [{
      sphereCollisionRadius=0
      boundBone {boneName=dyn_foot_l}
      boxExtents=<8 10 22>
      constraintSetup {angularLimitsMax=<5 12 12>, angularLimitsMin=<-5 -12 -12>, angularTarget=<0 0 0>}
      localJointOffset=<-8 -0.25 0>
     }]
   }
   n4 = SelectFloat(A=0, B=0.75, bPickA=Get IsAim() -> (IsAim:bool))
   n5 = Anim Dynamics - Bone: dyn_ear_l(ComponentPose=@n3, Alpha=@n4, ExternalForce="0,0,0") {
    angularDampingOverride=0.999
    angularSpringConstant=15
    bAngularSpring=true
    bGravityOverrideInSimSpace=true
    bOverrideAngularDamping=true
    bUseGravityOverride=true
    alphaBoolBlend {blendOption=Linear}
    boundBone {boneName=dyn_ear_l}
    componentAppliedLinearAccClamp=<200 200 200>
    componentLinearAccScale=<1 1 1>
    componentLinearVelScale=<1 1 1>
    gravityOverride {z=-0.5}
    physicsBodyDefinitions [{
      sphereCollisionRadius=0
      boundBone {boneName=dyn_ear_l}
      boxExtents=<2 6 6>
      constraintSetup {angularLimitsMax=<0 0 5>, angularLimitsMin=<0 0 -5>, angularTarget=<0 0 0>}
      localJointOffset=<-4 0 0>
     }]
   }
   n6 = Anim Dynamics - Bone: dyn_ear_r(ComponentPose=@n5, Alpha=@n4, ExternalForce="0,0,0") {
    angularDampingOverride=0.999
    angularSpringConstant=15
    bAngularSpring=true
    bGravityOverrideInSimSpace=true
    bOverrideAngularDamping=true
    bUseGravityOverride=true
    alphaBoolBlend {blendOption=Linear}
    boundBone {boneName=dyn_ear_r}
    componentAppliedLinearAccClamp=<200 200 200>
    componentLinearAccScale=<1 1 1>
    componentLinearVelScale=<1 1 1>
    gravityOverride {z=-0.5}
    physicsBodyDefinitions [{
      sphereCollisionRadius=0
      boundBone {boneName=dyn_ear_r}
      boxExtents=<4 4 6>
      constraintSetup {angularLimitsMax=<0 0 5>, angularLimitsMin=<0 0 -5>, angularTarget=<0 0 0>}
      localJointOffset=<-4 0 0>
     }]
   }
   Output Pose(Result=Component To Local(ComponentPose=@n6)) {layerGroup=None}
  ]}
 EventGraph {nodes [
   Event BlueprintUpdateAnimation()
   > Set IsAim(IsAim:bool=Get bIsTargeting()) -> (Output_Get:bool)
   TryGetPawnOwner()
  ]}
]
variables [IsAim {type=boolean}]
